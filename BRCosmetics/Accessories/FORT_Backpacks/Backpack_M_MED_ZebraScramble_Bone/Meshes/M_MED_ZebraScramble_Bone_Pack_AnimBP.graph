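# Anim Blueprint for the ZebraScramble Bone backpack, adding secondary motion to its ruffles, legs, tail, lock, medal and ears.
# A serial Anim Dynamics chain simulates each bone; two duplicate ruffle nodes sit disconnected.
# While skydiving without an open parachute, additive rotations pose the front-left leg and body.
# Customization load sets BackPackScale from the ZebraScramble meta tag; a bool-gated node ends the chain by additively scaling the pack bone.
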AnimBlueprint /BRCosmetics/Accessories/FORT_Backpacks/Backpack_M_MED_ZebraScramble_Bone/Meshes/M_MED_ZebraScramble_Bone_Pack_AnimBP
target_skeleton=&/M_MED_ZebraScramble_Bone_Pack_Skeleton
graphs [
 AnimGraph {nodes [
   n1 = Anim Dynamics - Bone: dyn_ruffles_2_l(Alpha=0.75, ExternalForce="0,0,0") {
    angularDampingOverride=0.95
    angularSpringConstant=1
    bGravityOverrideInSimSpace=true
    bOverrideAngularDamping=true
    bOverrideLinearDamping=true
    bUseGravityOverride=true
    bUsePlanarLimit=false
    linearDampingOverride=1
    linearSpringConstant=1
    alphaBoolBlend {blendOption=Linear}
    boundBone {boneName=dyn_ruffles_2_l}
    componentAppliedLinearAccClamp=<10 500 10>
    componentLinearAccScale=<1 1 1>
    componentLinearVelScale=<1 1 1>
    gravityOverride {z=-200}
    physicsBodyDefinitions [{
      boundBone {boneName=dyn_ruffles_2_l}
      boxExtents=<0.5 8 3>
      constraintSetup {angularLimitsMax=<0 90 0>, angularLimitsMin=<0 -90 0>, angularTarget=<0 0 0>}
      localJointOffset=<0 0 -3>
     }]
   }
   Anim Dynamics - Bone: dyn_ruffles_2_r(ComponentPose=@n1, Alpha=0.75, ExternalForce="0,0,0") {
    angularDampingOverride=0.95
    angularSpringConstant=1
    bGravityOverrideInSimSpace=true
    bOverrideAngularDamping=true
    bOverrideLinearDamping=true
    bUseGravityOverride=true
    bUsePlanarLimit=false
    linearDampingOverride=1
    linearSpringConstant=1
    alphaBoolBlend {blendOption=Linear}
    boundBone {boneName=dyn_ruffles_2_r}
    componentAppliedLinearAccClamp=<10 500 10>
    componentLinearAccScale=<1 1 1>
    componentLinearVelScale=<1 1 1>
    gravityOverride {z=-200}
    physicsBodyDefinitions [{
      boundBone {boneName=dyn_ruffles_2_r}
      boxExtents=<0.5 8 3>
      constraintSetup {angularLimitsMax=<0 90 0>, angularLimitsMin=<0 -90 0>, angularTarget=<0 0 0>}
      localJointOffset=<0 0 -3>
     }]
   }
   n2 = Anim Dynamics - Bone: dyn_ruffles_1_m(ComponentPose=Local To Component(LocalPose=Input Pose()), Alpha=0.75, ExternalForce="0,0,0") {
    angularDampingOverride=0.95
    angularSpringConstant=1
    bGravityOverrideInSimSpace=true
    bOverrideAngularDamping=true
    bOverrideLinearDamping=true
    bUseGravityOverride=true
    bUsePlanarLimit=false
    linearDampingOverride=1
    linearSpringConstant=1
    alphaBoolBlend {blendOption=Linear}
    boundBone {boneName=dyn_ruffles_1_m}
    componentAppliedLinearAccClamp=<1000 10 10>
    componentLinearAccScale=<1 1 1>
    componentLinearVelScale=<1 1 1>
    gravityOverride {z=-200}
    physicsBodyDefinitions [{
      boundBone {boneName=dyn_ruffles_1_m}
      boxExtents=<0.5 8 3>
      constraintSetup {angularLimitsMax=<0 90 0>, angularLimitsMin=<0 -90 0>, angularTarget=<0 0 0>}
      localJointOffset=<0 0 -3>
     }]
   }
   n3 = Anim Dynamics - Bone: dyn_ruffles_2_l(ComponentPose=@n2, Alpha=0.75, ExternalForce="0,0,0") {
    angularDampingOverride=0.95
    angularSpringConstant=1
    bGravityOverrideInSimSpace=true
    bOverrideAngularDamping=true
    bOverrideLinearDamping=true
    bUseGravityOverride=true
    bUsePlanarLimit=false
    linearDampingOverride=1
    linearSpringConstant=1
    alphaBoolBlend {blendOption=Linear}
    boundBone {boneName=dyn_ruffles_2_l}
    componentAppliedLinearAccClamp=<1000 10 10>
    componentLinearAccScale=<1 1 1>
    componentLinearVelScale=<1 1 1>
    gravityOverride {z=-200}
    physicsBodyDefinitions [{
      boundBone {boneName=dyn_ruffles_2_l}
      boxExtents=<0.5 8 3>
      constraintSetup {angularLimitsMax=<0 90 0>, angularLimitsMin=<0 -90 0>, angularTarget=<0 0 0>}
      localJointOffset=<0 0 -3>
     }]
   }
   n4 = Anim Dynamics - Bone: dyn_ruffles_2_r(ComponentPose=@n3, Alpha=0.75, ExternalForce="0,0,0") {
    angularDampingOverride=0.95
    angularSpringConstant=1
    bGravityOverrideInSimSpace=true
    bOverrideAngularDamping=true
    bOverrideLinearDamping=true
    bUseGravityOverride=true
    bUsePlanarLimit=false
    linearDampingOverride=1
    linearSpringConstant=1
    alphaBoolBlend {blendOption=Linear}
    boundBone {boneName=dyn_ruffles_2_r}
    componentAppliedLinearAccClamp=<1000 10 10>
    componentLinearAccScale=<1 1 1>
    componentLinearVelScale=<1 1 1>
    gravityOverride {z=-200}
    physicsBodyDefinitions [{
      boundBone {boneName=dyn_ruffles_2_r}
      boxExtents=<0.5 8 3>
      constraintSetup {angularLimitsMax=<0 90 0>, angularLimitsMin=<0 -90 0>, angularTarget=<0 0 0>}
      localJointOffset=<0 0 -3>
     }]
   }
   n5 = Anim Dynamics - Bone: dyn_ruffles_3_m(ComponentPose=@n4, Alpha=0.75, ExternalForce="0,0,0") {
    angularDampingOverride=0.95
    angularSpringConstant=1
    bGravityOverrideInSimSpace=true
    bOverrideAngularDamping=true
    bOverrideLinearDamping=true
    bUseGravityOverride=true
    bUsePlanarLimit=false
    linearDampingOverride=1
    linearSpringConstant=1
    alphaBoolBlend {blendOption=Linear}
    boundBone {boneName=dyn_ruffles_3_m}
    componentAppliedLinearAccClamp=<1000 10 10>
    componentLinearAccScale=<1 1 1>
    componentLinearVelScale=<1 1 1>
    gravityOverride {z=-200}
    physicsBodyDefinitions [{
      boundBone {boneName=dyn_ruffles_3_m}
      boxExtents=<0.5 8 3>
      constraintSetup {angularLimitsMax=<0 90 0>, angularLimitsMin=<0 -90 0>, angularTarget=<0 0 0>}
      localJointOffset=<0 0 -3>
     }]
   }
   n6 = Anim Dynamics - Bone: dyn_ruffles_4_m(ComponentPose=@n5, Alpha=0.5, ExternalForce="0,0,0") {
    angularDampingOverride=1
    angularSpringConstant=50
    bGravityOverrideInSimSpace=true
    bOverrideAngularDamping=true
    bOverrideLinearDamping=true
    bUseGravityOverride=true
    bUsePlanarLimit=false
    linearDampingOverride=0.25
    linearSpringConstant=10
    alphaBoolBlend {blendOption=Linear}
    boundBone {boneName=dyn_ruffles_4_m}
    componentAppliedLinearAccClamp=<100 100 100>
    componentLinearAccScale=<0.5 0.5 0.5>
    componentLinearVelScale=<0.5 0.5 0.5>
    gravityOverride {z=-100}
    physicsBodyDefinitions [{
      boundBone {boneName=dyn_ruffles_4_m}
      boxExtents=<1 5 1.5>
      constraintSetup {angularLimitsMax=<0 90 0>, angularLimitsMin=<0 -30 0>, angularTarget=<0 0 0>}
      localJointOffset=<0 0 -1.25>
     }]
   }
   n7 = Anim Dynamics - Bone: dyn_ruffles_5_l(ComponentPose=@n6, Alpha=0.5, ExternalForce="0,0,0") {
    angularDampingOverride=1
    angularSpringConstant=50
    bGravityOverrideInSimSpace=true
    bLinearSpring=true
    bOverrideAngularDamping=true
    bOverrideLinearDamping=true
    bUseGravityOverride=true
    bUsePlanarLimit=false
    linearDampingOverride=0.95
    linearSpringConstant=10
    alphaBoolBlend {blendOption=Linear}
    boundBone {boneName=dyn_ruffles_5_l}
    componentAppliedLinearAccClamp=<100 100 100>
    componentLinearAccScale=<0.5 0.5 0.5>
    componentLinearVelScale=<0.1 0.1 0.1>
    gravityOverride {z=-100}
    physicsBodyDefinitions [{
      boundBone {boneName=dyn_ruffles_5_l}
      boxExtents=<1 5 1>
      constraintSetup {angularLimitsMax=<10 90 0>, angularLimitsMin=<-10 -80 0>, angularTarget=<0 0 0>}
      localJointOffset=<0 0 -1>
     }]
   }
   n8 = Anim Dynamics - Bone: dyn_ruffles_6_l(ComponentPose=@n7, Alpha=1, ExternalForce="0,0,0") {
    angularDampingOverride=1
    angularSpringConstant=1
    bGravityOverrideInSimSpace=true
    bLinearSpring=true
    bOverrideAngularDamping=true
    bOverrideLinearDamping=true
    bUseGravityOverride=true
    bUsePlanarLimit=false
    linearDampingOverride=0.9
    linearSpringConstant=10
    alphaBoolBlend {blendOption=Linear}
    boundBone {boneName=dyn_ruffles_6_l}
    componentAppliedLinearAccClamp=<100 100 100>
    componentLinearAccScale=<0.75 0.75 0.75>
    componentLinearVelScale=<0.1 0.1 0.1>
    gravityOverride {z=-100}
    physicsBodyDefinitions [{
      boundBone {boneName=dyn_ruffles_6_l}
      boxExtents=<6 1 2>
      constraintSetup {twistAxis=AxisY, angularLimitsMax=<20 10 0>, angularLimitsMin=<-60 -10 0>, angularTarget=<0 0 0>}
      localJointOffset=<0 0 -1>
     }]
   }
   n9 = Anim Dynamics - Bone: dyn_ruffles_7_l(ComponentPose=@n8, Alpha=0.5, ExternalForce="0,0,0") {
    angularDampingOverride=1
    angularSpringConstant=1
    bGravityOverrideInSimSpace=true
    bLinearSpring=true
    bOverrideAngularDamping=true
    bOverrideLinearDamping=true
    bUseGravityOverride=true
    bUsePlanarLimit=false
    linearDampingOverride=0.85
    linearSpringConstant=10
    alphaBoolBlend {blendOption=Linear}
    boundBone {boneName=dyn_ruffles_7_l}
    componentAppliedLinearAccClamp=<100 100 100>
    componentLinearAccScale=<1 1 1>
    componentLinearVelScale=<0.5 0.5 0.5>
    gravityOverride {z=-100}
    physicsBodyDefinitions [{
      boundBone {boneName=dyn_ruffles_7_l}
      boxExtents=<5 1 1.5>
      constraintSetup {twistAxis=AxisY, angularLimitsMax=<90 0 20>, angularLimitsMin=<-45 0 -20>, angularTarget=<0 0 0>}
      localJointOffset=<0 0 -1.8>
     }]
   }
   n10 = Anim Dynamics - Bone: dyn_ruffles_12_l(ComponentPose=@n9, Alpha=0.75, ExternalForce="0,0,0") {
    angularDampingOverride=1
    angularSpringConstant=1
    bGravityOverrideInSimSpace=true
    bLinearSpring=true
    bOverrideAngularDamping=true
    bOverrideLinearDamping=true
    bUseGravityOverride=true
    bUsePlanarLimit=false
    linearDampingOverride=0.95
    linearSpringConstant=10
    alphaBoolBlend {blendOption=Linear}
    boundBone {boneName=dyn_ruffles_12_l}
    componentAppliedLinearAccClamp=<100 100 100>
    componentLinearAccScale=<0.75 0.75 0.75>
    componentLinearVelScale=<0.1 0.1 0.1>
    gravityOverride {z=-100}
    physicsBodyDefinitions [{
      boundBone {boneName=dyn_ruffles_12_l}
      boxExtents=<5 1 1.5>
      constraintSetup {twistAxis=AxisY, angularLimitsMax=<10 0 10>, angularLimitsMin=<-50 0 -10>, angularTarget=<0 0 0>}
      localJointOffset=<0 0 -1.8>
     }]
   }
   n11 = Anim Dynamics - Bone: dyn_ruffles_8_l(ComponentPose=@n10, Alpha=0.5, ExternalForce="0,0,0") {
    angularDampingOverride=1
    angularSpringConstant=10
    bAngularSpring=true
    bGravityOverrideInSimSpace=true
    bOverrideAngularDamping=true
    bOverrideLinearDamping=true
    bUseGravityOverride=true
    bUsePlanarLimit=false
    linearDampingOverride=1
    alphaBoolBlend {blendOption=Linear}
    boundBone {boneName=dyn_ruffles_8_l}
    componentAppliedLinearAccClamp=<100 100 100>
    componentLinearAccScale=<0.75 0.75 0.75>
    componentLinearVelScale=<0.25 0.25 0.25>
    gravityOverride {z=-100}
    physicsBodyDefinitions [{
      boundBone {boneName=dyn_ruffles_8_l}
      boxExtents=<5 1 1.5>
      constraintSetup {twistAxis=AxisY, angularLimitsMax=<45 0 20>, angularLimitsMin=<-45 0 -20>, angularTarget=<0 0 0>}
      localJointOffset=<0 0 -1>
     }]
   }
   n12 = Anim Dynamics - Bone: dyn_ruffles_9_m(ComponentPose=@n11, Alpha=0.5, ExternalForce="0,0,0") {
    angularDampingOverride=1
    angularSpringConstant=10
    bGravityOverrideInSimSpace=true
    bOverrideAngularDamping=true
    bOverrideLinearDamping=true
    bUseGravityOverride=true
    bUsePlanarLimit=false
    linearDampingOverride=0.25
    alphaBoolBlend {blendOption=Linear}
    boundBone {boneName=dyn_ruffles_9_m}
    componentAppliedLinearAccClamp=<100 100 100>
    componentLinearAccScale=<1 1 1>
    componentLinearVelScale=<0.5 0.5 0.5>
    gravityOverride {z=-150}
    physicsBodyDefinitions [{
      boundBone {boneName=dyn_ruffles_9_m}
      boxExtents=<3 5 1.5>
      constraintSetup {angularLimitsMax=<0 10 0>, angularLimitsMin=<0 -60 0>, angularTarget=<0 0 0>}
      localJointOffset=<1 0 1>
     }]
   }
   n13 = Anim Dynamics - Bone: dyn_ruffles_10_m(ComponentPose=@n12, Alpha=0.5, ExternalForce="0,0,0") {
    angularDampingOverride=0.5
    angularSpringConstant=10
    bGravityOverrideInSimSpace=true
    bOverrideAngularDamping=true
    bOverrideLinearDamping=true
    bUseGravityOverride=true
    bUsePlanarLimit=false
    linearDampingOverride=0.25
    alphaBoolBlend {blendOption=Linear}
    boundBone {boneName=dyn_ruffles_10_m}
    componentAppliedLinearAccClamp=<120 120 120>
    componentLinearAccScale=<0.8 0.8 0.8>
    componentLinearVelScale=<1 1 1>
    gravityOverride {z=-150}
    physicsBodyDefinitions [{
      boundBone {boneName=dyn_ruffles_10_m}
      boxExtents=<3 5 1.5>
      constraintSetup {angularLimitsMax=<0 20 0>, angularLimitsMin=<0 -60 0>, angularTarget=<0 0 0>}
      localJointOffset=<1.2 0 1>
     }]
   }
   n14 = Anim Dynamics - Bone: dyn_ruffles_11_m(ComponentPose=@n13, Alpha=0.5, ExternalForce="0,0,0") {
    angularDampingOverride=1
    angularSpringConstant=50
    bAngularSpring=true
    bGravityOverrideInSimSpace=true
    bOverrideAngularDamping=true
    bOverrideLinearDamping=true
    bUseGravityOverride=true
    bUsePlanarLimit=false
    linearDampingOverride=0.8
    alphaBoolBlend {blendOption=Linear}
    boundBone {boneName=dyn_ruffles_11_m}
    componentAppliedLinearAccClamp=<100 100 100>
    componentLinearAccScale=<0.5 0.5 0.5>
    componentLinearVelScale=<0.1 0.1 0.1>
    gravityOverride {z=-100}
    physicsBodyDefinitions [{
      boundBone {boneName=dyn_ruffles_11_m}
      boxExtents=<1 5 1.5>
      constraintSetup {angularLimitsMax=<0 60 0>, angularLimitsMin=<0 -60 0>, angularTarget=<0 0 0>}
      localJointOffset=<0 0 -1>
     }]
   }
   n15 = Anim Dynamics - Bone: dyn_leg_l_b!(ComponentPose=@n14, Alpha=0.25, GravityOverride:Vector="0,0,-60", ExternalForce="0,0,0") -> (Pose) {
    angularDampingOverride=1
    angularSpringConstant=5000
    bAngularSpring=true
    bGravityOverrideInSimSpace=true
    bLinearSpring=true
    bOverrideAngularDamping=true
    bOverrideLinearDamping=true
    bUseGravityOverride=true
    linearDampingOverride=0.95
    linearSpringConstant=300
    alphaBoolBlend {blendOption=Linear}
    boundBone {boneName=dyn_leg_l_b}
    componentAppliedLinearAccClamp=<350 350 350>
    componentLinearAccScale=<0.75 0.75 0.75>
    componentLinearVelScale=<0.25 0.25 0.25>
    gravityOverride {z=-60}
    physicsBodyDefinitions [{
      boundBone {boneName=dyn_leg_l_b}
      boxExtents=<5 2 2>
      constraintSetup {angularLimitsMax=<0 45 45>, angularLimitsMin=<0 -45 -45>, angularTarget=<0 0 0>}
      localJointOffset=<1.75 0 0>
     }]
   }
   n16 = Anim Dynamics - Bone: dyn_leg_r_b!(ComponentPose=@n15, Alpha=0.33, GravityOverride:Vector="0,0,-60", ExternalForce="0,0,0") -> (Pose) {
    angularDampingOverride=1
    angularSpringConstant=5000
    bAngularSpring=true
    bGravityOverrideInSimSpace=true
    bLinearSpring=true
    bOverrideAngularDamping=true
    bOverrideLinearDamping=true
    bUseGravityOverride=true
    linearDampingOverride=0.95
    linearSpringConstant=300
    alphaBoolBlend {blendOption=Linear}
    boundBone {boneName=dyn_leg_r_b}
    componentAppliedLinearAccClamp=<350 350 350>
    componentLinearAccScale=<0.75 0.75 0.75>
    componentLinearVelScale=<0.25 0.25 0.25>
    gravityOverride {z=-60}
    physicsBodyDefinitions [{
      boundBone {boneName=dyn_leg_r_b}
      boxExtents=<5 2 2>
      constraintSetup {angularLimitsMax=<0 45 45>, angularLimitsMin=<0 -45 -45>, angularTarget=<0 0 0>}
      localJointOffset=<1.75 0 0>
     }]
   }
   n17 = Anim Dynamics - Bone: dyn_leg_r_f!(ComponentPose=@n16, Alpha=0.25, GravityOverride:Vector="0,0,-60", ExternalForce="0,0,0") -> (Pose) {
    angularDampingOverride=1
    angularSpringConstant=5000
    bAngularSpring=true
    bGravityOverrideInSimSpace=true
    bLinearSpring=true
    bOverrideAngularDamping=true
    bOverrideLinearDamping=true
    bUseGravityOverride=true
    linearDampingOverride=0.95
    linearSpringConstant=300
    alphaBoolBlend {blendOption=Linear}
    boundBone {boneName=dyn_leg_r_f}
    componentAppliedLinearAccClamp=<350 350 350>
    componentLinearAccScale=<0.75 0.75 0.75>
    componentLinearVelScale=<0.25 0.25 0.25>
    gravityOverride {z=-60}
    physicsBodyDefinitions [{
      boundBone {boneName=dyn_leg_r_f}
      boxExtents=<5 2 2>
      constraintSetup {angularLimitsMax=<0 45 45>, angularLimitsMin=<0 -45 -45>, angularTarget=<0 0 0>}
      localJointOffset=<1.75 0 0>
     }]
   }
   n18 = Anim Dynamics - Bone: dyn_leg_l_f!(ComponentPose=@n17, Alpha=0.33, GravityOverride:Vector="0,0,-65", ExternalForce="0,0,0") -> (Pose) {
    angularDampingOverride=1
    angularSpringConstant=5000
    bAngularSpring=true
    bGravityOverrideInSimSpace=true
    bLinearSpring=true
    bOverrideAngularDamping=true
    bOverrideLinearDamping=true
    bUseGravityOverride=true
    linearDampingOverride=0.95
    linearSpringConstant=300
    alphaBoolBlend {blendOption=Linear}
    boundBone {boneName=dyn_leg_l_f}
    componentAppliedLinearAccClamp=<350 350 350>
    componentLinearAccScale=<0.75 0.75 0.75>
    componentLinearVelScale=<0.25 0.25 0.25>
    gravityOverride {z=-65}
    physicsBodyDefinitions [{
      boundBone {boneName=dyn_leg_l_f}
      boxExtents=<5 2 2>
      constraintSetup {angularLimitsMax=<0 45 45>, angularLimitsMin=<0 -45 -45>, angularTarget=<0 0 0>}
      localJointOffset=<1.75 0 0>
     }]
   }
   n19 = Get Dyn_Leg_Rotation() -> (Dyn_Leg_Rotation:Rotator)
   n20 = Anim Dynamics - Bone: dyn_tail(
    ComponentPose="Transform (Modify) Bone - Bone: dyn_body"(
     ComponentPose="Transform (Modify) Bone - Bone: dyn_leg_l_f"(ComponentPose=@n18, Alpha=1, Translation="0,0,0", Rotation=@n19, Scale="1,1,1") {rotationMode=BMM_Additive, alphaBoolBlend {blendOption=Linear}, boneToModify {boneName=dyn_leg_l_f}}
     Alpha=1
     Translation="0,0,0"
     Rotation=Get Dyn_Body_Rotation() -> (Dyn_Body_Rotation:Rotator)
     Scale="1,1,1"
    ) {rotationMode=BMM_Additive, alphaBoolBlend {blendOption=Linear}, boneToModify {boneName=dyn_body}}
    Alpha=0.75
    ExternalForce="0,0,0"
   ) {
    angularDampingOverride=0.95
    angularSpringConstant=1
    bAngularSpring=true
    bGravityOverrideInSimSpace=true
    bOverrideAngularDamping=true
    bUseGravityOverride=true
    bUsePlanarLimit=false
    linearSpringConstant=1000
    alphaBoolBlend {blendOption=Linear}
    boundBone {boneName=dyn_tail}
    componentAppliedLinearAccClamp=<60 60 60>
    componentLinearAccScale=<0.75 0.75 0.75>
    componentLinearVelScale=<0.25 0.25 0.25>
    gravityOverride {z=-45}
    physicsBodyDefinitions [{
      boundBone {boneName=dyn_tail}
      boxExtents=<4.5 2 2>
      constraintSetup {angularLimitsMax=<0 30 10>, angularLimitsMin=<0 -15 -10>, angularTarget=<0 0 0>}
      localJointOffset=<2 0 0.75>
     }]
   }
   n21 = Anim Dynamics - Bone: dyn_lock(ComponentPose=@n20, Alpha=0.6, ExternalForce="0,0,0") {
    angularDampingOverride=0.3
    angularSpringConstant=1
    bAngularSpring=true
    bGravityOverrideInSimSpace=true
    bLinearSpring=true
    bOverrideAngularDamping=true
    bUseGravityOverride=true
    linearDampingOverride=0.7
    linearSpringConstant=1
    alphaBoolBlend {blendOption=Linear}
    boundBone {boneName=dyn_lock}
    componentAppliedLinearAccClamp=<750 750 750>
    componentLinearAccScale=<0.5 0.5 0.5>
    componentLinearVelScale=<0.5 0.5 0.5>
    gravityOverride {z=-980}
    physicsBodyDefinitions [{
      boundBone {boneName=dyn_lock}
      boxExtents=<3 5 7.5>
      constraintSetup {twistAxis=AxisY, angularLimitsMax=<0 20 0>, angularLimitsMin=<-100 -20 0>, angularTarget=<0 0 0>}
      localJointOffset=<0 0 -1>
     }]
   }
   n22 = Anim Dynamics - Bone: dyn_medal(ComponentPose=@n21, Alpha=0.75, ExternalForce="0,0,0") {
    angularDampingOverride=1
    angularSpringConstant=10
    bAngularSpring=true
    bGravityOverrideInSimSpace=true
    bLinearSpring=true
    bOverrideAngularDamping=true
    bUseGravityOverride=true
    linearSpringConstant=1
    alphaBoolBlend {blendOption=Linear}
    boundBone {boneName=dyn_medal}
    componentAppliedLinearAccClamp=<100 100 100>
    componentLinearAccScale=<1 1 1>
    componentLinearVelScale=<1 1 1>
    gravityOverride {z=-120}
    physicsBodyDefinitions [{
      boundBone {boneName=dyn_medal}
      boxExtents=<6 2 2>
      constraintSetup {angularLimitsMax=<0 90 60>, angularLimitsMin=<0 0 -60>, angularTarget=<0 0 0>}
      localJointOffset=<3 0 0>
     }]
   }
   n23 = Anim Dynamics - Bone: dyn_ear_l(ComponentPose=@n22, Alpha=1, ExternalForce="0,0,0") {
    angularDampingOverride=1
    angularSpringConstant=10
    bAngularSpring=true
    bGravityOverrideInSimSpace=true
    bLinearSpring=true
    bOverrideAngularDamping=true
    bUseGravityOverride=true
    linearSpringConstant=1
    alphaBoolBlend {blendOption=Linear}
    boundBone {boneName=dyn_ear_l}
    componentAppliedLinearAccClamp=<100 100 100>
    componentLinearAccScale=<1 1 1>
    componentLinearVelScale=<1 1 1>
    gravityOverride {z=-120}
    physicsBodyDefinitions [{
      boundBone {boneName=dyn_ear_l}
      boxExtents=<6 2 2>
      constraintSetup {angularLimitsMax=<0 20 60>, angularLimitsMin=<0 -20 0>, angularTarget=<0 0 0>}
      localJointOffset=<3 0 0>
     }]
   }
   n24 = Anim Dynamics - Bone: dyn_ear_r(ComponentPose=@n23, Alpha=1, ExternalForce="0,0,0") {
    angularDampingOverride=1
    angularSpringConstant=10
    bAngularSpring=true
    bGravityOverrideInSimSpace=true
    bLinearSpring=true
    bOverrideAngularDamping=true
    bUseGravityOverride=true
    linearSpringConstant=1
    alphaBoolBlend {blendOption=Linear}
    boundBone {boneName=dyn_ear_r}
    componentAppliedLinearAccClamp=<100 100 100>
    componentLinearAccScale=<1 1 1>
    componentLinearVelScale=<1 1 1>
    gravityOverride {z=-120}
    physicsBodyDefinitions [{
      boundBone {boneName=dyn_ear_r}
      boxExtents=<6 2 2>
      constraintSetup {angularLimitsMax=<0 10 0>, angularLimitsMin=<0 -10 -60>, angularTarget=<0 0 0>}
      localJointOffset=<3 0 0>
     }]
   }
   n25 = "Transform (Modify) Bone - Bone: pack"!(ComponentPose=@n24, bAlphaBoolEnabled:bool, Translation="0,0,0", Rotation="0,0,0", Scale="1.4,1.4,1.4") -> (Pose) {
    alphaInputType=Bool
    rotationMode=BMM_Additive
    scaleMode=BMM_Additive
    translationMode=BMM_Additive
    alphaBoolBlend {blendOption=Linear}
    boneToModify {boneName=pack}
    scale=<1.4 1.4 1.4>
   }
   Output Pose(Result=Component To Local(ComponentPose=@n25)) {layerGroup=None}
  ]}
 EventGraph {nodes [
   Event BlueprintUpdateAnimation()
   n1 = Get bIsParachuteOpened()
   > n2 = Branch(Condition=AND Boolean(A=Get bIsSkyDiving(), B=NOT Boolean(A=@n1)))
   > Set Dyn_Leg_Rotation(Dyn_Leg_Rotation:Rotator="-20,0,20") -> (Output_Get:Rotator)
   > Set Dyn_Body_Rotation(Dyn_Body_Rotation:Rotator="0,0,10") -> (Output_Get:Rotator)
   Set Dyn_Leg_Rotation(execute=@n2.else, Dyn_Leg_Rotation="0, 0, 0")
   > Set Dyn_Body_Rotation(Dyn_Body_Rotation="0, 0, 0")
   Event OnCharacterCustomizationLoaded()
   > Set BackPackScale(BackPackScale:bool=CharacterItemDefinitionHasAnyMetaTag(Tags=Make Array("[0]"=(TagName="Cosmetics.MetaTags.ZebraScramble")))) -> (Output_Get:bool)
   TryGetPawnOwner()
  ]}
]
variables [
 BackPackScale {type=boolean}
 Dyn_Body_Rotation {type=Rotator}
 Dyn_Leg_Rotation {type=Rotator}
]
